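# Physics asset for the M_MED_Backspin backpack cosmetic, defining physics bodies for its pack and sword bones.
# Each bone has a single box shape: a squarish one for the pack and a longer, narrow one for the sword.
# One constraint joins the sword to the pack with limited swing and twist, restricting how far the sword can rotate relative to the pack.
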PhysicsAsset /BRCosmetics/Accessories/FORT_Backpacks/Backpack_M_MED_Backspin/Meshes/M_MED_Backspin_Pack_Physics
bodies {
 pack [{shapeType=Box, extentX=10, extentY=2, extentZ=10, center=<0 0.8989 -2.4482>}]
 sword [{shapeType=Box, extentX=2, extentY=20, extentZ=7}]
}
constraints [{bone1Name=sword, bone2Name=pack, swing1LimitDegrees=30, swing2LimitDegrees=10, twistLimitDegrees=30}]
