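# Animation Blueprint for the Backspin backpack, adding physics motion and toggling a material-driven element to match the equipped pickaxe.
# The AnimGraph runs a RigidBody node on the ref pose using the pack's physics asset, simulated in root-bone space.
# Character load, weapon changes, sprint and skydive events drive "Hide Element 01" on the dynamic material.
# It is 1 while a BackSpin-tagged pickaxe is equipped, and 0 otherwise or upon starting sprint or skydive.
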AnimBlueprint /BRCosmetics/Accessories/FORT_Backpacks/Backpack_M_MED_Backspin/Meshes/M_MED_Backspin_Pack_AnimBP
target_skeleton=&/M_MED_Backspin_Pack_Skeleton
graphs [
 AnimGraph {nodes [
   n1 = RigidBody(ComponentPose=Local To Component(LocalPose=Local Space Ref Pose())) {
    bClampLinearTranslationLimitToRefPose=true
    bFreezeIncomingPoseOnStart=true
    bOverrideWorldGravity=true
    bTransferBoneVelocities=true
    evaluationResetTime=8
    overridePhysicsAsset=&/M_MED_Backspin_Pack_Physics
    simulationSpace=BaseBoneSpace
    worldSpaceMinimumScale=8
    alphaBoolBlend {blendOption=Linear}
    baseBoneRef {boneName=root}
    componentLinearAccScale=<1 0.3 0.3>
    componentLinearVelScale=<1 0.3 0.3>
    overrideWorldGravity {z=-980}
   }
   Output Pose(Result=Component To Local(ComponentPose=@n1)) {layerGroup=None}
  ]}
 EventGraph {nodes [
   n1 = Event OnBeginTacticalSprinting() {is=title K2Node_Event}
   n2 = Get Dynamic Material Instance(execute=[@n6.then, @n9.then], SkeletalMeshComponent=GetPartSkeletalMeshComponent())
   SetScalarParameterValue(execute=[@n1.then, @n7.then, @n11.then, @n5.else, @n12.else], self=@n2, ParameterName=Hide Element 01, Value=0)
   n3 = Event OnEndTacticalSprinting() {is=title K2Node_Event}
   n4 = PickaxeItemDefinitionHasAnyMetaTag(Tags=Make Array("[0]"=(TagName="Cosmetics.MetaTags.BackSpin")), bRequiresEquipped=true)
   n5 = Branch(execute=[@n3.then, @n8.then], Condition=@n4)
   SetScalarParameterValue(execute=[@n11.else, @n5.then], self=@n2, ParameterName=Hide Element 01, Value=1)
   n6 = Event OnWeaponChanged()
   n7 = Event OnBeginSkydiving()
   n8 = Event OnEndSkydiving()
   n9 = Event OnCharacterCustomizationLoaded()
   n10 = OR Boolean(A=Get bIsSkyDiving(), B=Get bIsTacticalSprinting())
   n11 = Branch(execute=@n12.then, Condition=@n10)
   n12 = Branch(execute=@n2.then, Condition=@n4)
  ]}
]
variables [
 Is_BackPack_Equipped {type=boolean, default=true}
 PickaxeEquipped {type=number}
]
